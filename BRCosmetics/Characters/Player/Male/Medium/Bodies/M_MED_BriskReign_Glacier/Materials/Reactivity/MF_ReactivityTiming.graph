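# Material function for reactivity timing on the BriskReign Glacier body cosmetic, producing a scalar envelope from a start time and total duration.
# Duration ratios split the time into fade-in, plateau and fade-out phases, with fade-out taking the remainder.
# Three MF_TimeCurve calls drive the phases, combined by multiplies and a lerp using peak and plateau exit values.
# A second output displays four scalars through DebugFloat4Values.
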MaterialFunction /BRCosmetics/Characters/Player/Male/Medium/Bodies/M_MED_BriskReign_Glacier/Materials/Reactivity/MF_ReactivityTiming
graph {expressions {
  MaterialExpressionAdd_10 {
   class=MaterialExpressionAdd
   inputs [
    A {from=MaterialExpressionNamedRerouteUsage_170}
    B {from=MaterialExpressionNamedRerouteUsage_171}
   ]
  }
  MaterialExpressionAdd_3 {
   class=MaterialExpressionAdd
   inputs [
    A {from=MaterialExpressionNamedRerouteUsage_141}
    B {from=MaterialExpressionMultiply_19}
   ]
  }
  MaterialExpressionAdd_5 {
   class=MaterialExpressionAdd
   inputs [
    A {from=MaterialExpressionNamedRerouteUsage_146}
    B {from=MaterialExpressionNamedRerouteUsage_145}
   ]
  }
  MaterialExpressionConstant_0 {class=MaterialExpressionConstant}
  MaterialExpressionConstant_2 {class=MaterialExpressionConstant, properties {r=1}}
  MaterialExpressionConstant_3 {class=MaterialExpressionConstant}
  MaterialExpressionConstant_4 {class=MaterialExpressionConstant, properties {r=1}}
  MaterialExpressionConstant_6 {class=MaterialExpressionConstant, properties {r=2}}
  MaterialExpressionConstant_7 {class=MaterialExpressionConstant, properties {r=3}}
  MaterialExpressionConvert_10 {
   class=MaterialExpressionConvert
   inputs [
    None {from=MaterialExpressionNamedRerouteUsage_1}
    None {from=MaterialExpressionNamedRerouteUsage_0}
    None {from=MaterialExpressionNamedRerouteUsage_2}
    None {from=MaterialExpressionNamedRerouteUsage_3}
   ]
   properties {
    nodeName=Make Vector4
    convertInputs [
     {
      type=Scalar
      expressionInput {
       expression=&self:MaterialExpressionNamedRerouteUsage_1
       inputName=None
       mask=0
       maskA=0
       maskB=0
       maskG=0
       maskR=0
       outputIndex=0
      }
     }
     {
      type=Scalar
      expressionInput {
       expression=&self:MaterialExpressionNamedRerouteUsage_0
       inputName=None
       mask=0
       maskA=0
       maskB=0
       maskG=0
       maskR=0
       outputIndex=0
      }
     }
     {
      type=Scalar
      expressionInput {
       expression=&self:MaterialExpressionNamedRerouteUsage_2
       inputName=None
       mask=0
       maskA=0
       maskB=0
       maskG=0
       maskR=0
       outputIndex=0
      }
     }
     {
      type=Scalar
      expressionInput {
       expression=&self:MaterialExpressionNamedRerouteUsage_3
       inputName=None
       mask=0
       maskA=0
       maskB=0
       maskG=0
       maskR=0
       outputIndex=0
      }
     }
    ]
    convertOutputs [{type=Vector4, defaultValue=<rgba 0 0 0 1>}]
   }
  }
  MaterialExpressionFunctionInput_0 {
   class=MaterialExpressionFunctionInput
   inputs [Preview {from="", output=""}]
   properties {inputName=Start Time, inputType=FunctionInput_Scalar}
  }
  MaterialExpressionFunctionInput_1 {
   class=MaterialExpressionFunctionInput
   inputs [Preview {from="", output=""}]
   properties {inputName=Total Duration, inputType=FunctionInput_Scalar, sortPriority=10}
  }
  MaterialExpressionFunctionInput_2 {
   class=MaterialExpressionFunctionInput
   inputs [Preview {from="", output=""}]
   properties {bUsePreviewValueAsDefault=true, inputName=Fade-In Duration Ratio, inputType=FunctionInput_Scalar, sortPriority=20, previewValue=<0.15 0 0 1>}
  }
  MaterialExpressionFunctionInput_3 {
   class=MaterialExpressionFunctionInput
   inputs [Preview {from="", output=""}]
   properties {bUsePreviewValueAsDefault=true, inputName=Plateau Duration Ratio, inputType=FunctionInput_Scalar, sortPriority=40, previewValue=<0.2 0 0 1>}
  }
  MaterialExpressionFunctionInput_4 {
   class=MaterialExpressionFunctionInput
   inputs [Preview {from="", output=""}]
   properties {bUsePreviewValueAsDefault=true, inputName=Peak Value, inputType=FunctionInput_Scalar, sortPriority=30, previewValue=<1 0 0 1>}
  }
  MaterialExpressionFunctionInput_5 {
   class=MaterialExpressionFunctionInput
   inputs [Preview {from="", output=""}]
   properties {bUsePreviewValueAsDefault=true, inputName=Plateau Exit Value, inputType=FunctionInput_Scalar, sortPriority=50, previewValue=<0.9 0 0 1>}
  }
  MaterialExpressionFunctionOutput_0 {
   class=MaterialExpressionFunctionOutput
   inputs [None {from=MaterialExpressionNamedRerouteDeclaration_3, output=""}]
   properties {outputName="Result (S)"}
  }
  MaterialExpressionFunctionOutput_1 {
   class=MaterialExpressionFunctionOutput
   inputs [None {from=MaterialExpressionMaterialFunctionCall_9, output=GreyScaleOutput}]
   properties {outputName="Debug (S)", sortPriority=10}
  }
  MaterialExpressionLinearInterpolate_0 {
   class=MaterialExpressionLinearInterpolate
   inputs [
    A {from=MaterialExpressionMultiply_29}
    B {from=MaterialExpressionNamedRerouteUsage_160}
    Alpha {from=MaterialExpressionNamedRerouteUsage_6}
   ]
  }
  MaterialExpressionMaterialFunctionCall_4 {
   class=MaterialExpressionMaterialFunctionCall
   inputs [
    Time {from=""}
    Start Time {from=MaterialExpressionNamedRerouteUsage_138}
    Duration {from=MaterialExpressionNamedRerouteUsage_158}
    Direction {from=MaterialExpressionNamedRerouteUsage_8}
    Curve Type {from=MaterialExpressionNamedRerouteUsage_11}
   ]
   properties {desc=Fade-In, materialFunction=&/MF_TimeCurve}
  }
  MaterialExpressionMaterialFunctionCall_5 {
   class=MaterialExpressionMaterialFunctionCall
   inputs [
    Time {from=""}
    Start Time {from=MaterialExpressionNamedRerouteUsage_147}
    Duration {from=MaterialExpressionNamedRerouteUsage_157}
    Direction {from=MaterialExpressionNamedRerouteUsage_9}
    Curve Type {from=MaterialExpressionNamedRerouteUsage_13}
   ]
   properties {desc=Fade-Out, materialFunction=&MF_TimeCurve}
  }
  MaterialExpressionMaterialFunctionCall_8 {
   class=MaterialExpressionMaterialFunctionCall
   inputs [
    Time {from=""}
    Start Time {from=MaterialExpressionNamedRerouteUsage_175}
    Duration {from=MaterialExpressionNamedRerouteUsage_176}
    Direction {from=MaterialExpressionNamedRerouteUsage_10}
    Curve Type {from=MaterialExpressionNamedRerouteUsage_12}
   ]
   properties {desc=Plateau, materialFunction=&MF_TimeCurve}
  }
  MaterialExpressionMaterialFunctionCall_9 {
   class=MaterialExpressionMaterialFunctionCall
   inputs [
    Vector4 {from=MaterialExpressionConvert_10}
    MaximumNumberOfDigits {from=""}
    UVs {from=""}
    DebugTextLocation RG_UpperRight BA_LowerLeft {from=""}
    Component Spacing {from=""}
   ]
   properties {materialFunction=/Engine/Functions/Engine_MaterialFunctions02/Utility/DebugFloat4Values}
  }
  MaterialExpressionMultiply_16 {
   class=MaterialExpressionMultiply
   inputs [
    A {from=MaterialExpressionNamedRerouteUsage_144}
    B {from=MaterialExpressionNamedRerouteUsage_142}
   ]
   properties {constB=0.2}
  }
  MaterialExpressionMultiply_19 {
   class=MaterialExpressionMultiply
   inputs [
    A {from=MaterialExpressionAdd_5}
    B {from=MaterialExpressionNamedRerouteUsage_136}
   ]
   properties {constB=0.3}
  }
  MaterialExpressionMultiply_26 {
   class=MaterialExpressionMultiply
   inputs [
    A {from=MaterialExpressionOneMinus_2}
    B {from=MaterialExpressionNamedRerouteUsage_136}
   ]
   properties {constB=0.7}
  }
  MaterialExpressionMultiply_29 {
   class=MaterialExpressionMultiply
   inputs [
    A {from=MaterialExpressionNamedRerouteUsage_5}
    B {from=MaterialExpressionNamedRerouteUsage_159}
   ]
  }
  MaterialExpressionMultiply_32 {
   class=MaterialExpressionMultiply
   inputs [
    A {from=MaterialExpressionNamedRerouteUsage_173}
    B {from=MaterialExpressionNamedRerouteUsage_174}
   ]
   properties {constB=0.2}
  }
  MaterialExpressionMultiply_33 {
   class=MaterialExpressionMultiply
   inputs [
    A {from=MaterialExpressionLinearInterpolate_0}
    B {from=MaterialExpressionNamedRerouteUsage_7}
   ]
  }
  MaterialExpressionNamedRerouteDeclaration_0 {
   class=MaterialExpressionNamedRerouteDeclaration
   inputs [None {from=MaterialExpressionMaterialFunctionCall_4, output=Time}]
   properties {name="Fade-In Alpha (S)", nodeColor=<rgba 0 0.8828 1 1>}
  }
  MaterialExpressionNamedRerouteDeclaration_1 {
   class=MaterialExpressionNamedRerouteDeclaration
   inputs [None {from=MaterialExpressionMaterialFunctionCall_8, output=Time}]
   properties {name="Plateau Alpha (S)", nodeColor=<rgba 0.7578 0 1 1>}
  }
  MaterialExpressionNamedRerouteDeclaration_110 {
   class=MaterialExpressionNamedRerouteDeclaration
   inputs [None {from=MaterialExpressionFunctionInput_0}]
   properties {name="Start Time (S)", nodeColor=<rgba 0 1 0.3672 1>}
  }
  MaterialExpressionNamedRerouteDeclaration_111 {
   class=MaterialExpressionNamedRerouteDeclaration
   inputs [None {from=MaterialExpressionFunctionInput_1}]
   properties {name="Total Duration (S)", nodeColor=<rgba 0.5938 1 0 1>}
  }
  MaterialExpressionNamedRerouteDeclaration_112 {
   class=MaterialExpressionNamedRerouteDeclaration
   inputs [None {from=MaterialExpressionNamedRerouteUsage_36}]
   properties {name="Fade-In Start Time (S)", nodeColor=<rgba 1 0.2812 0 1>}
  }
  MaterialExpressionNamedRerouteDeclaration_113 {
   class=MaterialExpressionNamedRerouteDeclaration
   inputs [None {from=MaterialExpressionFunctionInput_2}]
   properties {name="Fade-In Duration Ratio (S)", nodeColor=<rgba 1 0 0.0469 1>}
  }
  MaterialExpressionNamedRerouteDeclaration_114 {
   class=MaterialExpressionNamedRerouteDeclaration
   inputs [None {from=MaterialExpressionFunctionInput_3}]
   properties {name="Plateau Duration Ratio (S)", nodeColor=<rgba 0 1 0.6719 1>}
  }
  MaterialExpressionNamedRerouteDeclaration_123 {
   class=MaterialExpressionNamedRerouteDeclaration
   inputs [None {from=MaterialExpressionFunctionInput_5}]
   properties {name="Plateau Exit Value (S)", nodeColor=<rgba 0 1 0.1094 1>}
  }
  MaterialExpressionNamedRerouteDeclaration_125 {
   class=MaterialExpressionNamedRerouteDeclaration
   inputs [None {from=MaterialExpressionAdd_3}]
   properties {name="Fade-Out Start Time (S)", nodeColor=<rgba 0.2656 0 1 1>}
  }
  MaterialExpressionNamedRerouteDeclaration_126 {
   class=MaterialExpressionNamedRerouteDeclaration
   inputs [None {from=MaterialExpressionMultiply_26}]
   properties {name="Fade-Out Duration (S)", nodeColor=<rgba 0.0547 1 0 1>}
  }
  MaterialExpressionNamedRerouteDeclaration_127 {
   class=MaterialExpressionNamedRerouteDeclaration
   inputs [None {from=MaterialExpressionMultiply_16}]
   properties {name="Fade-In Duration (S)", nodeColor=<rgba 1 0 0.375 1>}
  }
  MaterialExpressionNamedRerouteDeclaration_128 {
   class=MaterialExpressionNamedRerouteDeclaration
   inputs [None {from=MaterialExpressionFunctionInput_4}]
   properties {name="Peak Value (S)", nodeColor=<rgba 0 1 0.1094 1>}
  }
  MaterialExpressionNamedRerouteDeclaration_129 {
   class=MaterialExpressionNamedRerouteDeclaration
   inputs [None {from=MaterialExpressionAdd_10}]
   properties {name="Plateau Start Time (S)", nodeColor=<rgba 0.2656 0 1 1>}
  }
  MaterialExpressionNamedRerouteDeclaration_130 {
   class=MaterialExpressionNamedRerouteDeclaration
   inputs [None {from=MaterialExpressionMultiply_32}]
   properties {name="Plateau Duration (S)", nodeColor=<rgba 1 0 0.375 1>}
  }
  MaterialExpressionNamedRerouteDeclaration_2 {
   class=MaterialExpressionNamedRerouteDeclaration
   inputs [None {from=MaterialExpressionMaterialFunctionCall_5, output=Time}]
   properties {name="Fade-Out Curve (S)", nodeColor=<rgba 0 1 0.4375 1>}
  }
  MaterialExpressionNamedRerouteDeclaration_3 {
   class=MaterialExpressionNamedRerouteDeclaration
   inputs [None {from=MaterialExpressionMultiply_33}]
   properties {name="Out Value (S)", nodeColor=<rgba 1 0.1172 0 1>}
  }
  MaterialExpressionNamedRerouteDeclaration_4 {
   class=MaterialExpressionNamedRerouteDeclaration
   inputs [None {from=MaterialExpressionConstant_0}]
   properties {name="Forward Direction (S)", nodeColor=<rgba 0 0.2031 1 1>}
  }
  MaterialExpressionNamedRerouteDeclaration_5 {
   class=MaterialExpressionNamedRerouteDeclaration
   inputs [None {from=MaterialExpressionConstant_2}]
   properties {name="Reverse Direction (S)", nodeColor=<rgba 0.5234 1 0 1>}
  }
  MaterialExpressionNamedRerouteDeclaration_6 {
   class=MaterialExpressionNamedRerouteDeclaration
   inputs [None {from=MaterialExpressionConstant_3}]
   properties {name="Linear Curve (S)", nodeColor=<rgba 1 0 0.8437 1>}
  }
  MaterialExpressionNamedRerouteDeclaration_7 {
   class=MaterialExpressionNamedRerouteDeclaration
   inputs [None {from=MaterialExpressionConstant_4}]
   properties {name="Ease In Curve (S)", nodeColor=<rgba 0 1 0.8359 1>}
  }
  MaterialExpressionNamedRerouteDeclaration_8 {
   class=MaterialExpressionNamedRerouteDeclaration
   inputs [None {from=MaterialExpressionConstant_6}]
   properties {name="Ease Out Curve (S)", nodeColor=<rgba 1 0.5156 0 1>}
  }
  MaterialExpressionNamedRerouteDeclaration_9 {
   class=MaterialExpressionNamedRerouteDeclaration
   inputs [None {from=MaterialExpressionConstant_7}]
   properties {name="S Curve (S)", nodeColor=<rgba 0.1953 0 1 1>}
  }
  MaterialExpressionNamedRerouteUsage_0 {class=MaterialExpressionNamedRerouteUsage}
  MaterialExpressionNamedRerouteUsage_1 {class=MaterialExpressionNamedRerouteUsage}
  MaterialExpressionNamedRerouteUsage_10 {class=MaterialExpressionNamedRerouteUsage}
  MaterialExpressionNamedRerouteUsage_11 {class=MaterialExpressionNamedRerouteUsage}
  MaterialExpressionNamedRerouteUsage_12 {class=MaterialExpressionNamedRerouteUsage}
  MaterialExpressionNamedRerouteUsage_13 {class=MaterialExpressionNamedRerouteUsage}
  MaterialExpressionNamedRerouteUsage_136 {class=MaterialExpressionNamedRerouteUsage}
  MaterialExpressionNamedRerouteUsage_138 {class=MaterialExpressionNamedRerouteUsage}
  MaterialExpressionNamedRerouteUsage_141 {class=MaterialExpressionNamedRerouteUsage}
  MaterialExpressionNamedRerouteUsage_142 {class=MaterialExpressionNamedRerouteUsage}
  MaterialExpressionNamedRerouteUsage_144 {class=MaterialExpressionNamedRerouteUsage}
  MaterialExpressionNamedRerouteUsage_145 {class=MaterialExpressionNamedRerouteUsage}
  MaterialExpressionNamedRerouteUsage_146 {class=MaterialExpressionNamedRerouteUsage}
  MaterialExpressionNamedRerouteUsage_147 {class=MaterialExpressionNamedRerouteUsage}
  MaterialExpressionNamedRerouteUsage_157 {class=MaterialExpressionNamedRerouteUsage}
  MaterialExpressionNamedRerouteUsage_158 {class=MaterialExpressionNamedRerouteUsage}
  MaterialExpressionNamedRerouteUsage_159 {class=MaterialExpressionNamedRerouteUsage}
  MaterialExpressionNamedRerouteUsage_160 {class=MaterialExpressionNamedRerouteUsage}
  MaterialExpressionNamedRerouteUsage_170 {class=MaterialExpressionNamedRerouteUsage}
  MaterialExpressionNamedRerouteUsage_171 {class=MaterialExpressionNamedRerouteUsage}
  MaterialExpressionNamedRerouteUsage_173 {class=MaterialExpressionNamedRerouteUsage}
  MaterialExpressionNamedRerouteUsage_174 {class=MaterialExpressionNamedRerouteUsage}
  MaterialExpressionNamedRerouteUsage_175 {class=MaterialExpressionNamedRerouteUsage}
  MaterialExpressionNamedRerouteUsage_176 {class=MaterialExpressionNamedRerouteUsage}
  MaterialExpressionNamedRerouteUsage_2 {class=MaterialExpressionNamedRerouteUsage}
  MaterialExpressionNamedRerouteUsage_3 {class=MaterialExpressionNamedRerouteUsage}
  MaterialExpressionNamedRerouteUsage_36 {class=MaterialExpressionNamedRerouteUsage}
  MaterialExpressionNamedRerouteUsage_5 {class=MaterialExpressionNamedRerouteUsage}
  MaterialExpressionNamedRerouteUsage_6 {class=MaterialExpressionNamedRerouteUsage}
  MaterialExpressionNamedRerouteUsage_7 {class=MaterialExpressionNamedRerouteUsage}
  MaterialExpressionNamedRerouteUsage_8 {class=MaterialExpressionNamedRerouteUsage}
  MaterialExpressionNamedRerouteUsage_9 {class=MaterialExpressionNamedRerouteUsage}
  MaterialExpressionOneMinus_2 {class=MaterialExpressionOneMinus, inputs [None {from=MaterialExpressionAdd_5}]}
 }}
