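# Character-style master material for the DandyFlare backpack cosmetic, with an optional reactive emissive look.
# MF_CharacterParent_2023 PreFX output feeds PostFX through the Enable Reactivity static switch, which can blend in material attributes with a squared tint as emissive.
# Reactivity timing comes from a BriskReign_Glacier function fed timestamp, fade-in, plateau and intensity parameters.
# Blueprint-allow and force-look controls, a KillCount parameter and a selectable tech-art mask channel are exposed.
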Material /BRCosmetics/Accessories/FORT_Backpacks/Backpack_M_MED_DandyFlare/Materials/M_DandyFlare_Backpack
graph {
 properties {
  bAutomaticallySetUsageInEditor=false
  bUseMaterialAttributes=true
  bUsedWithMeshDeformer=true
  bUsedWithMorphTargets=true
  bUsedWithSkeletalMesh=true
  pixelDepthOffsetMode=PDOM_Legacy
  shadingModel=MSM_FromMaterialExpression
  usedShadingModels=Default Lit | Subsurface | Clear Coat | Subsurface Profile | Cloth
 }
 nodes [
  n1 = MaterialFunctionCall() {materialFunction=&MF_CharacterParent_2023_PreFX}
  n2 = NamedRerouteDeclaration(None=@n1.Result) {name="Material In (MA)", nodeColor=<rgba 0.5703 0 1 1>}
  n3 = VectorParameter() {
   group=Reactivity - Look
   parameterName=Reactivity - Emissive Color
   sortPriority=30
   channelNames {a=Intensity, b=Tint B, g=Tint G, r=Tint R}
   defaultValue=<rgba 1 1 1 1>
  }
  n4 = SetMaterialAttributes(MaterialAttributes=@n2, Emissive Color=Multiply(A=@n3.RGB, B=@n3.RGB)) {attributeSetTypes [#1]}
  n5 = Multiply(A=NamedRerouteUsage(), B=NamedRerouteUsage())
  MaterialFunctionCall(
   Material=StaticSwitchParameter(True=BlendMaterialAttributes(A=@n2, B=@n4, Alpha=@n5), False=NamedRerouteUsage()) {group=Reactivity - Timing, parameterName=Enable Reactivity, sortPriority=0}
   M=@n1.M
   Pre Skinned Local Position=@n1.M
   Distance Blend=@n1.M
   Fresnel=@n1.M
  ) {materialFunction=&MF_CharacterParent_2023_PostFX}
  NamedRerouteDeclaration(None=ScalarParameter() {group="Reactivity - BP (do not edit)", parameterName=Timestamp, sortPriority=10}) {name="Reactivity - Start Time (BP) (S)", nodeColor=<rgba 0 1 0.3672 1>}
  n6 = VectorParameter() {
   group=Reactivity - Timing
   parameterName=Reactivity - Duration
   sortPriority=110
   channelNames {a=n/a, b=Plateau Duration Ratio, g=Fade-In Duration Ratio, r=Total Duration}
   defaultValue=<rgba 1 0.1 0.25 0>
  }
  NamedRerouteDeclaration(None=@n6.Total Duration) {name="Reactivity - Total Duration (S)", nodeColor=<rgba 0.5938 1 0 1>}
  NamedRerouteDeclaration(None=@n6.Fade-In Duration Ratio) {name="Reactivity - Fade-In Duration Ratio (S)", nodeColor=<rgba 1 0 0.0469 1>}
  NamedRerouteDeclaration(None=@n6.Plateau Duration Ratio) {name="Reactivity - Plateau Duration Ratio (S)", nodeColor=<rgba 0 1 0.6719 1>}
  n7 = VectorParameter() {
   group=Reactivity - Timing
   parameterName=Reactivity - Intensity
   sortPriority=120
   channelNames {a=n/a, b=n/a, g=Plateau Exit Value, r=Peak Value}
   defaultValue=<rgba 1 0.9 0 0>
  }
  NamedRerouteDeclaration(None=@n7.Plateau Exit Value) {name="Reactivity - Plateau Exit Value (S)", nodeColor=<rgba 0 1 0.1094 1>}
  NamedRerouteDeclaration(None=@n7.Peak Value) {name="Reactivity - Peak Value (S)", nodeColor=<rgba 0 1 0.1094 1>}
  NamedRerouteDeclaration(None=ScalarParameter() {group=Reactivity - Timing, parameterName=Reactivity - Allow BP Reactivity, sliderMax=1, sortPriority=20}) {name="Reactivity - Allow BP Reactivity (S)", nodeColor=<rgba 0 0.2031 1 1>}
  n8 = ScalarParameter() {
   desc=bypasses all parameters in this group and activates all reactivity overrides
   group=Reactivity - Timing
   parameterName=Reactivity - Force Reactive Look
   sliderMax=1
   sortPriority=10
  }
  NamedRerouteDeclaration(None=@n8) {name="Reactivity - Force Reactive Look (S)", nodeColor=<rgba 0.5234 1 0 1>}
  n9 = ChannelMaskParameter(None=NamedRerouteDeclaration(None=@n1.Tech Art Mask) {name="TechArt Mask (V3)", nodeColor=<rgba 0.6261 0 1 1>}) {group=Reactivity - Look, maskChannel=Green, parameterName=ReactivityMaskChannel, sortPriority=10, defaultValue=<rgba 0 1 0 0>}
  NamedRerouteDeclaration(None=@n9.R) {name="Reactivity Mask (S)", nodeColor=<rgba 1 0 0.5625 1>}
  NamedRerouteDeclaration(None=ScalarParameter() {group=Reactivity - Timing, parameterName=KillCount, sortPriority=40}) {name="KillCount (S)", nodeColor=<rgba 1 0.8906 0 1>}
  n10 = NamedRerouteUsage()
  n11 = MaterialFunctionCall!(
   Start Time=NamedRerouteUsage()
   Total Duration=NamedRerouteUsage()
   Fade-In Duration Ratio=NamedRerouteUsage()
   Peak Value=NamedRerouteUsage()
   Plateau Duration Ratio=NamedRerouteUsage()
   Plateau Exit Value=NamedRerouteUsage()
  ) {materialFunction=/BRCosmetics/Characters/Player/Male/Medium/Bodies/M_MED_BriskReign_Glacier/Materials/Reactivity/MF_ReactivityTiming}
  NamedRerouteDeclaration(None=LinearInterpolate(A=Multiply(A=@n10, B=@n11."Result (S)"), Alpha=NamedRerouteUsage())) {name="Reactivity Alpha (S)", nodeColor=<rgba 0.0547 1 0 1>}
 ]
}
