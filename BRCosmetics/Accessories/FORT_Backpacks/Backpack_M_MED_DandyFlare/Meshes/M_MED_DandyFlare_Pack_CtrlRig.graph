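# Control Rig for the DandyFlare backpack that spins its drum_01 bone, apparently in response to enemy eliminations.
# Each evaluation reads the bone's X Euler rotation and adds bEnemyPlayerEliminated, a double despite its boolean-style name, to form a target angle.
# A critically damped spring interpolation eases toward that target.
# The resulting angle becomes an X-only quaternion that feeds a Modify Transforms on drum_01.
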ControlRigBlueprint /BRCosmetics/Accessories/FORT_Backpacks/Backpack_M_MED_DandyFlare/Meshes/M_MED_DandyFlare_Pack_CtrlRig
graphs [
 RigVMFunctionLibrary {links [], nodes []}
 RigVMModel {nodes [
   n1 = &RigUnit_BeginExecution()
   RigVMFunction_MathQuaternionToEuler = &RigVMFunction_MathQuaternionToEuler(RotationOrder=ZYX, Value=&RigUnit_GetTransform(Item=(Type=Bone,Name="drum_01"), bInitial=False).Transform.Rotation)
   SpringInterp = /Script/ControlRig.RigUnit_SpringInterpV2(
    CriticalDamping=1
    Force=0
    Strength=4
    TargetVelocityAmount=0
    bInitializeFromTarget=False
    bUseCurrentInput=False
    Target=&RigVMFunction_MathDoubleAdd(B=@RigVMFunction_MathQuaternionToEuler.Result.X, A=RigVMVariableNode() {variable=bEnemyPlayerEliminated}.Value).Result
    Current=@RigVMFunction_MathQuaternionToEuler.Result.X
   ) {
    is=struct RigVMUnitNode
    template="SpringInterp::Execute(in Target,in Strength,in CriticalDamping,in Force,in bUseCurrentInput,in Current,in TargetVelocityAmount,in bInitializeFromTarget,out Result,out Velocity)"
   }
   RigVMFunction_MathQuaternionFromEuler = &RigVMFunction_MathQuaternionFromEuler(Euler=(X=0,Y=0,Z=0), Euler.X=@SpringInterp.Result)
   &RigUnit_ModifyTransforms(
    ItemToModify=((Item=(Type=Bone,Name="drum_01"),Transform=(Rotation=(X=0,Y=0,Z=0,W=1),Translation=(X=0,Y=0,Z=0),Scale3D=(X=1,Y=1,Z=1))))
    Weight=1
    WeightMinimum=0
    ExecutePin=@n1.ExecutePin
    ItemToModify.0.Transform.Rotation=@RigVMFunction_MathQuaternionFromEuler.Result
   )
  ]}
]
hierarchy [
 root {loc=<0 0 0>, rot=<0 0 0 1>}
 pack {parent=root, loc=<0 -0.7316 0>, rot=<0 0 0 1>}
 drum_main {parent=pack, loc=<0.0067 -10.8008 -0.5746>, rot=<0 0 0.7071 -0.7071>}
 drum_01 {parent=drum_main, loc=<0.5558 0 0>, rot=<0 0 0 1>}
]
variables [bEnemyPlayerEliminated {type=double, default=0}]
