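# Animation Blueprint for the DandyFlare backpack cosmetic, driving pack placement, a Control Rig animation and tube physics.
# Spine yaw relative to its reference pose additively offsets and rolls the pack bone, feeding the Control Rig.
# The rig's bEnemyPlayerEliminated input is a float stepped on anim input events once a watched kill stat rises, or by a looping timer in the front end.
# Anim Dynamics on the tube bone is blended in only in the front end.
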AnimBlueprint /BRCosmetics/Accessories/FORT_Backpacks/Backpack_M_MED_DandyFlare/Meshes/M_MED_DandyFlare_Pack_AnimBP
target_skeleton=&/M_MED_DandyFlare_Pack_Skeleton
graphs [
 AnimGraph {nodes [
   n1 = Anim Dynamics - Bone: tube(ComponentPose=Local To Component(LocalPose=Use cached pose 'Main pose'()), Alpha=1, ExternalForce="0,0,0") {
    angularDampingOverride=1
    angularSpringConstant=8000
    bAngularSpring=true
    bOverrideAngularDamping=true
    boundBone {boneName=tube}
    componentAppliedLinearAccClamp=<3000 3000 2500>
    componentLinearAccScale=<0.5 0.5 0.5>
    componentLinearVelScale=<0.5 0.5 0.5>
    physicsBodyDefinitions [{
      boundBone {boneName=tube}
      boxExtents=<20 20 20>
      constraintSetup {angularLimitsMax=<0 10 0>, angularLimitsMin=<0 -30 0>, angularTarget=<1 0 0>, linearAxesMax=<0.1 0 0>}
      localJointOffset=<0.1 0 0>
     }]
    simSpaceSettings {maxAngularAcceleration=35, maxAngularVelocity=7, simSpaceAngularAlpha=1, externalAngularVelocity {y=5}}
   }
   Output Pose(Result=Blend Poses by bool(BlendPose_0=Component To Local(ComponentPose=@n1), BlendPose_1=Use cached pose 'Main pose'(), bActiveValue=Get bIsInFrontEnd())) {layerGroup=None}
   n2 = "Transform (Modify) Bone - Bone: pack"(ComponentPose=Local To Component(LocalPose=Input Pose()), Translation=Get backpack_translation(), Rotation=Get backpack_rotate()) {rotationMode=BMM_Additive, translationMode=BMM_Additive, alphaBoolBlend {blendOption=Linear}, boneToModify {boneName=pack}}
   n3 = Control Rig!(
    Source=Component To Local(ComponentPose=@n2)
    Alpha
    bEnemyPlayerEliminated:double=Get bEnemyPlayerEliminated() -> (bEnemyPlayerEliminated:double)
   ) -> (Pose) {controlRigAssetReference {blueprintRigClass=&/M_MED_DandyFlare_Pack_CtrlRig.M_MED_DandyFlare_Pack_CtrlRig_C}}
   Save cached pose 'Main pose'(Pose=@n3) {cacheName=Main pose}
  ]}
 EventGraph {nodes [
   n1 = CustomEvent() {is=title K2Node_CustomEvent}
   > n2 = Branch(Condition=Get bIsInFrontEnd())
   Set bEnemyPlayerEliminated(execute=[@n7.then, @n2.then], bEnemyPlayerEliminated:double=float + float(A=Get bEnemyPlayerEliminated(), B=-45)) -> (Output_Get:double)
   Event BlueprintUpdateAnimation()
   > n3 = Sequence!(execute) -> (then_0, then_1, then_2:Exec, then_3:Exec)
   n4 = Branch(execute=@n3.then_1, Condition=Get bIsInFrontEnd())
   > Set SimlationSpace(SimlationSpace:AnimPhysSimSpaceType Enum=Component) -> (Output_Get:AnimPhysSimSpaceType Enum)
   Set SimlationSpace(execute=@n4.else)
   n5 = GetOwnerSkeletalMeshComponent()
   n6 = BreakRotator(InRot=BreakTransform(InTransform=GetDeltaTransformFromRefPose(self=@n5, BoneName=spine_05, BaseName=spine_01)).Rotation)
   Set backpack_rotate(execute=@n3.then_0, backpack_rotate=MakeRotator(Roll=MapRangeClamped(Value=@n6.Yaw, InRangeA=-30, InRangeB=-150, OutRangeB=240)))
   > Set backpack_translation(backpack_translation=MakeVector(
     Y=MapRangeClamped(Value=@n6.Yaw, InRangeA=-20, InRangeB=-50, OutRangeB=-17)
     Z=MapRangeClamped(Value=@n6.Yaw, InRangeA=-30, InRangeB=-50, OutRangeB=-2)
    ))
   Event BlueprintProcessAnimInputEvent()
   > n7 = Branch(Condition=Get Active_Kill() -> (Active_Kill:bool))
   n8 = Event OnCharacterCustomizationLoaded()
   Sequence!(execute=[@n8.then, @n12.then]) -> (then_0, then_1)
   > n9 = Cast To FortPlayerPawn(Object=TryGetFortPawn())
   > n10 = WatchCosmeticStat!(
    execute:Exec
    Pawn:Fort Player Pawn Object Reference=@n9
    StatTag:Fort Stat Manager Tag Structure=(TagName="StatManager.AthenaKills")
    DefaultValue:int=0
    bBroadcastInitialValue:bool=false
    InitialValueBroadcastDelay:float=-1
    ChangeValueBroadcastDelay:float=-1
    bConsumeWaitingNotifications:bool=true
   ) -> (then:Exec, OnStatChanged:Exec, StatName:Name, StatValue:int) {is=title K2Node_AsyncAction}
   n11 = Set KillCount(execute=@n10.OnStatChanged, KillCount:double="To Float (Integer)"(InInt=@n10.StatValue)) -> (Output_Get:double)
   > Branch(Condition="float > float"(A=@n11, B=Get Last_KillCount() -> (Last_KillCount:double)) {is=title K2Node_CallFunction})
   > Set Active_Kill(Active_Kill:bool=true) -> (Output_Get:bool)
   > Set Last_KillCount(Last_KillCount:double=@n11) -> (Output_Get:double)
   Event BlueprintBeginPlay() {is=title K2Node_Event}
   > Set Timer by Event(Delegate=@n1, Time=8, bLooping=true)
   n12 = Event BlueprintInitializeAnimation()
   IsNamed(InEventName=aie.EnemyPlayerEliminated)
   TryGetPawnOwner()
  ]}
 OnRep_bEnemyPlayerEliminated {nodes [OnRep_bEnemyPlayerEliminated() {is=title K2Node_FunctionEntry}]}
]
variables [
 Active_Kill {type=boolean}
 backpack_rotate {type=Rotator}
 backpack_translation {type=Vector}
 bEnemyPlayerEliminated {type=number}
 KillCount {type=number}
 Last_KillCount {type=number}
 SimlationSpace {type=AnimPhysSimSpaceType}
]
